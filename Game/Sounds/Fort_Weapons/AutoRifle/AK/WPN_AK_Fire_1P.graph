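# First-person gunfire sound cue for the AK auto rifle, built from heavy assault rifle samples.
# A mixer layers two random three-variant wave sets, with the v5 set at half input volume.
# Each layer passes through a modulator with pitch held fixed, and both random nodes are excluded from branch culling.
# It uses first-person weapon attenuation, sound class and gunfire concurrency, and is flagged for debug.
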SoundCue /Game/Sounds/Fort_Weapons/AutoRifle/AK/WPN_AK_Fire_1P
root=Mixer_0
nodes [
 {id=Mixer_0, children [Modulator_1,Modulator_0], properties {inputVolume [1,0.5]}}
 {id=Modulator_0, children [Random_4], properties {pitchMax=1, pitchMin=1}}
 {id=Modulator_1, children [Random_3], properties {pitchMax=1, pitchMin=1, volumeMax=1, volumeMin=1}}
 {id=Random_3, children [WavePlayer_0,WavePlayer_2,WavePlayer_1], properties {bShouldExcludeFromBranchCulling=true, weights [1,1,1]}}
 {id=Random_4, children [WavePlayer_11, WavePlayer_15, WavePlayer_17], properties {bShouldExcludeFromBranchCulling=true, weights [1,1,1]}}
 {id=WavePlayer_0, properties {soundWaveAssetPtr=&/WPN_HeavyAR_LT_01}}
 {id=WavePlayer_1, properties {soundWaveAssetPtr=&/WPN_HeavyAR_LT_02}}
 {id=WavePlayer_11, properties {soundWaveAssetPtr=&/WPN_HeavyAR_LT_v5_01}}
 {id=WavePlayer_15, properties {soundWaveAssetPtr=&/WPN_HeavyAR_LT_v5_02}}
 {id=WavePlayer_17, properties {soundWaveAssetPtr=&/WPN_HeavyAR_LT_v5_03}}
 {id=WavePlayer_2, properties {soundWaveAssetPtr=&/WPN_HeavyAR_LT_03}}
]
properties {
 attenuationSettings=&Fort_Weapon_Projectile_FirstPerson_Athena_Att
 bDebug=true
 maxDistance=2097152
 soundClassObject=&Fort_WeaponRanged_Fire_1P_PSM
 volumeMultiplier=1
 concurrencySet [&Weapon_Gunfire_FirstPerson_Conc]
}
