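# Third-person AK fire SoundCue that layers close, medium and distant gunshots, each band with its own attenuation.
# The close band mixes two randomized HeavyAR layers under the 3P PSM sound class.
# Medium and distant bands reuse randomized AssaultHeavy waves with modulation, mixed down under the Distant fire class.
# Local and global gunfire concurrency sets limit playback, random nodes are excluded from branch culling, and the debug flag is on.
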SoundCue /Game/Sounds/Fort_Weapons/AutoRifle/AK/WPN_AK_Fire_3P
root=Mixer_1
nodes [
 {id=Attenuation_118, children [Mixer_2], properties {attenuationSettings=&Fort_Weapon_Projectile_Close_Athena_Att}}
 {
  id=Attenuation_119
  children [Modulator_2]
  properties {attenuationSettings=&Fort_Weapon_Projectile_Medium_Mono_Athena_Att}
 }
 {
  id=Attenuation_120
  children [Modulator_1]
  properties {attenuationSettings=&Fort_Weapon_Projectile_Distant_Mono_Athena_Att}
 }
 {id=Mixer_0, children [Attenuation_119, Attenuation_120], properties {inputVolume [0.215,0.25]}}
 {id=Mixer_1, children [SoundClass_77, SoundClass_76], properties {inputVolume [0.7,1]}}
 {id=Mixer_2, children [Modulator_3,Modulator_0], properties {inputVolume [1,0.5]}}
 {id=Modulator_0, children [Random_4], properties {pitchMax=1, pitchMin=1}}
 {id=Modulator_1, children [Random_1], properties {pitchMax=1.1, pitchMin=0.9, volumeMax=0.8, volumeMin=0.7}}
 {id=Modulator_2, children [Random_0], properties {volumeMax=2, volumeMin=1.9}}
 {id=Modulator_3, children [Random_2], properties {pitchMax=1, pitchMin=1, volumeMax=1, volumeMin=1}}
 {id=Random_0, children [WavePlayer_0,WavePlayer_1,WavePlayer_2], properties {bShouldExcludeFromBranchCulling=true, weights [1,1,1]}}
 {id=Random_1, children [WavePlayer_3,WavePlayer_4], properties {bShouldExcludeFromBranchCulling=true, weights [1,1]}}
 {id=Random_2, children [WavePlayer_5,WavePlayer_7,WavePlayer_6], properties {bShouldExcludeFromBranchCulling=true, weights [1,1,1]}}
 {id=Random_4, children [WavePlayer_11, WavePlayer_15, WavePlayer_17], properties {bShouldExcludeFromBranchCulling=true, weights [1,1,1]}}
 {id=SoundClass_76, children [Mixer_0], properties {soundClassOverride=&Fort_WeaponRanged_Fire_Distant}}
 {
  id=SoundClass_77
  children [Attenuation_118]
  properties {soundClassOverride=&Fort_WeaponRanged_Fire_3P_PSM}
 }
 {id=WavePlayer_0, properties {soundWaveAssetPtr=/Game/Sounds/Fort_Weapons/AutoRifle/AssaultHeavy/AssaultHeavy_Fire_MediumNew_01}}
 {id=WavePlayer_1, properties {soundWaveAssetPtr=/Game/Sounds/Fort_Weapons/AutoRifle/AssaultHeavy/AssaultHeavy_Fire_MediumNew_03}}
 {id=WavePlayer_11, properties {soundWaveAssetPtr=&/WPN_HeavyAR_LT_v5_01}}
 {id=WavePlayer_15, properties {soundWaveAssetPtr=&/WPN_HeavyAR_LT_v5_02}}
 {id=WavePlayer_17, properties {soundWaveAssetPtr=&/WPN_HeavyAR_LT_v5_03}}
 {id=WavePlayer_2, properties {soundWaveAssetPtr=/Game/Sounds/Fort_Weapons/AutoRifle/AssaultHeavy/AssaultHeavy_Fire_MediumNew_02}}
 {id=WavePlayer_3, properties {soundWaveAssetPtr=/Game/Sounds/Fort_Weapons/AutoRifle/AssaultHeavy/AssaultHeavy_Fire_DistantNew_01}}
 {id=WavePlayer_4, properties {soundWaveAssetPtr=/Game/Sounds/Fort_Weapons/AutoRifle/AssaultHeavy/AssaultHeavy_Fire_DistantNew_02}}
 {id=WavePlayer_5, properties {soundWaveAssetPtr=&/WPN_HeavyAR_LT_01}}
 {id=WavePlayer_6, properties {soundWaveAssetPtr=&/WPN_HeavyAR_LT_02}}
 {id=WavePlayer_7, properties {soundWaveAssetPtr=&/WPN_HeavyAR_LT_03}}
]
properties {
 bDebug=true
 maxDistance=27500
 pitchMultiplier=1.1
 soundClassObject=&Fort_WeaponRanged
 volumeMultiplier=1
 concurrencySet [&Weapon_Gunfire_Local_Conc, &Weapon_Gunfire_Global_Conc]
}
